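# Physics asset for the male stand-alone Headband pickaxe, named as the variant used while the axe is swinging.
# A long capsule covers the two-handed handle bone, and two short capsules cover the dyn_Handle_1 and dyn_Handle_2 bones, likely the dangling headband pieces.
# Each dyn bone is constrained directly to the handle with free swing and twist, so its stored angular limits do not restrict rotation.
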PhysicsAsset /BRCosmetics/Weapons/FORT_Melee/Pickaxe_Headband_Stand_Alone_Male/Meshes/Headband_Stand_Alone_Male_Axe_IsSwinging_Physics
bodies {
 Melee_TwoHanded_Handle [{shapeType=Capsule, length=60, radius=2, center=<0 -0.9945 0>}]
 dyn_Handle_1 [{
   shapeType=Capsule
   length=5
   radius=3
   shapeName=dyn_Handle_1_capsule
   center=<7.7709 -1.8378 0.0026>
   rotation=<rot -0.029 -107.8126 -89.9755>
  }]
 dyn_Handle_2 [{shapeType=Capsule, length=5, radius=3, shapeName=dyn_Handle_2_capsule, center=<10.3664 0.4496 0.1143>, rotation=<rot 1.6433 93.9497 -85.479>}]
}
constraints [
 {
  bone1Name=dyn_Handle_1
  bone2Name=Melee_TwoHanded_Handle
  swing1LimitDegrees=45
  swing1Motion=Free
  swing2LimitDegrees=45
  swing2Motion=Free
  twistLimitDegrees=20
  twistMotion=Free
 }
 {
  bone1Name=dyn_Handle_2
  bone2Name=Melee_TwoHanded_Handle
  swing1LimitDegrees=45
  swing1Motion=Free
  swing2LimitDegrees=45
  swing2Motion=Free
  twistLimitDegrees=20
  twistMotion=Free
 }
]
