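# Animation Blueprint for the Headband Stand Alone Male pickaxe that adds physics-driven secondary motion to its skeleton.
# The input pose runs through a base-bone-space RigidBody node, whose strength drops in the front end, then a world-space RigidBody using a separate IsSwinging physics asset.
# The second node has a bool alpha with timed linear blending, but no visible logic drives it, as the event graph holds only unconnected nodes.
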AnimBlueprint /BRCosmetics/Weapons/FORT_Melee/Pickaxe_Headband_Stand_Alone_Male/Meshes/Headband_Stand_Alone_Male_Axe_AnimBP
target_skeleton=&/Headband_Stand_Alone_Male_Axe_Skeleton
graphs [
 AnimGraph {nodes [
   AnimGraphNode_SubInput_0 = Input Pose()
   n1 = RigidBody(ComponentPose=Local To Component(LocalPose=@AnimGraphNode_SubInput_0), Alpha=SelectFloat(A=0.3, B=1, bPickA=Get bIsInFrontEnd())) {
    bClampLinearTranslationLimitToRefPose=true
    bOverrideWorldGravity=true
    overridePhysicsAsset=&/Headband_Stand_Alone_Male_Axe_Physics
    simulationSpace=BaseBoneSpace
    alphaBoolBlend {blendOption=Linear}
    baseBoneRef {boneName=b_AR_Root}
    componentLinearAccScale=<0.4 0.4 0.4>
    componentLinearVelScale=<0.2 0.2 0.2>
    overrideWorldGravity {z=-980}
   }
   n2 = RigidBody!(ComponentPose=@n1, bAlphaBoolEnabled:bool, ExternalForce, bTransferBoneVelocities:bool=False) -> (Pose) {
    alphaInputType=Bool
    bOverrideWorldGravity=true
    overridePhysicsAsset=&/Headband_Stand_Alone_Male_Axe_IsSwinging_Physics
    simulationSpace=WorldSpace
    alphaBoolBlend {blendInTime=0.5, blendOption=Linear, blendOutTime=0.5}
    baseBoneRef {boneName=b_AR_Root}
    componentLinearAccScale=<0.4 0.4 0.4>
    componentLinearVelScale=<0.2 0.2 0.2>
    overrideWorldGravity {z=-980}
   }
   Output Pose(Result=Component To Local(ComponentPose=@n2)) {layerGroup=None}
  ]}
 EventGraph {nodes [
   TryGetPawnOwner()
   Event BlueprintUpdateAnimation()
  ]}
]
variables []
